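# Mono MetaSound source for a vehicle scraping sound, routed to the exterior vehicles sound class with close-detail movement attenuation.
# On play, a wave player starts a scrape loop asset, and its audio is multiplied by the Speed input, which therefore scales volume rather than pitch.
# A concurrency override limits it to one instance per owner.
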MetaSoundSource /VehicleAudio/MetaSounds/MSS_Vehicle_Scraping
inputs [
 Speed {type=Float}
 UE.Source.OnPlay {type=Trigger}
]
interfaces [UE.OutputFormat.Mono:1.0, UE.Source:1.1]
outputs ["UE.OutputFormat.Mono.Audio:0" {type=Audio}]
pages [Default {
  input_defaults {Speed=1, UE.Source.OnPlay=null}
  nodes [
   WavePlayer_0 = UE.Wave Player:Mono(Play=^UE.Source.OnPlay) {also {defaults {Loop="<MetaSoundBuilderResult.SUCCEEDED: 0>", Wave Asset=/VehicleAudio/Assets/Collision/Veh_Valet_Scrape_Loop_01}, version [1,1]}}
   Multiply_0 = UE.Multiply:Audio by Float(AdditionalOperands=^Speed, PrimaryOperand=@WavePlayer_0.Out Mono)
  ]
  links [{dst="@UE.OutputFormat.Mono.Audio:0", src=Multiply_0.Out}]
 }]
properties {
 assetImportData=&self:AssetImportData
 attenuationSettings=/VehicleAudio/Settings/Vehicle_Movement_Close_Detail_Att
 bLooping=true
 bOverrideConcurrency=true
 duration=10000
 maxDistance=12000
 sampleRate=48000
 soundClassObject=&Fort_Vehicles_Exterior
 concurrencyOverrides {bLimitToOwner=true, maxCount=1}
}
